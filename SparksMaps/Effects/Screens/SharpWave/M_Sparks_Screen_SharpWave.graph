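# Unlit Sparks stage screen material showing animated sharp-wave streaks with stage primary/secondary colors and a beat-driven reactivity value.
# Three channels of one wave texture pan at different speeds over noise-distorted UVs to form the streaks.
# A noise-masked dot grid, faint cartoon-flame background and corner glow add layers.
# Beat lifetime drives a curve-atlas remap for reactivity, and the summed emissive feeds the shared screen-details function.
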Material /SparksMaps/Effects/Screens/SharpWave/M_Sparks_Screen_SharpWave
graph {
 properties {bAutomaticallySetUsageInEditor=false, bHasPixelAnimation=true, shadingModel=MSM_Unlit, usedShadingModels=Unlit}
 nodes [
  n1 = TextureSample(UVs=Panner(Coordinate=TextureCoordinate()) {speedX=0.25, speedY=0.25}) {samplerType=SAMPLERTYPE_Grayscale, texture=&LowResBlurredNoise}
  n2 = TextureCoordinate()
  n3 = Constant() {r=65}
  n4 = SphereMask(A=Frac(None=Multiply(A=@n2, B=@n3) {constB=100}), B=Constant() {r=0.5}, Radius=Constant() {r=0.2}) {attenuationRadius=0.1, hardnessPercent=1}
  n5 = NamedRerouteUsage()
  n6 = OneMinus(None=@n5)
  n7 = Multiply(A=Multiply(A=Power(Base=@n1.R), B=Multiply(A=@n4) {constB=2}), B=@n6)
  n8 = Add(A=NamedRerouteUsage()) {constB=0.125}
  n9 = Multiply(A=Multiply(A=NamedRerouteUsage(), B=NamedRerouteUsage()), B=@n6)
  n10 = Multiply(A=NamedRerouteUsage(), B=@n5)
  n11 = NamedRerouteUsage()
  n12 = NamedRerouteUsage()
  n13 = Saturate(None=OneMinus(None=Add(A=@n11, B=@n12)))
  n14 = Multiply(A=Add(A=NamedRerouteUsage()) {constB=0.1}, B=NamedRerouteUsage())
  n15 = Multiply(A=Add(A=Add(A=@n9, B=@n10), B=Multiply(A=@n13, B=@n14))) {constB=0.5}
  n16 = NamedRerouteUsage()
  n17 = NamedRerouteUsage()
  n18 = Add(A=Add(A=NamedRerouteUsage(), B=NamedRerouteUsage()), B=NamedRerouteUsage())
  n19 = Rotator(
   Coordinate=LinearInterpolate(A=TextureCoordinate(), B=NamedRerouteUsage(), Alpha=ScalarParameter() {defaultValue=0.05, parameterName=Background Distort})
   Time=Constant() {r=-3}
  )
  n20 = AppendVector(A=Constant() {r=1}, B=Constant() {r=0.25})
  n21 = Constant()
  n22 = Constant() {r=0.1}
  n23 = TextureSample(UVs=Panner(Coordinate=Multiply(A=@n19, B=@n20), Speed=AppendVector(A=@n21, B=@n22)) {speedY=0.013}) {texture=/Game/Effects/Fort_Effects/Textures/Tile/T_Tile_Texture_Cartoon_Flames}
  n24 = Multiply(A=OneMinus(None=Saturate(None=@n18)), B=Saturate(None=Power(Base=@n23.R) {constExponent=1}))
  n25 = Add(A=Multiply(A=Multiply(A=@n15, B=@n16), B=NamedRerouteUsage()), B=Multiply(A=Add(A=@n17) {constB=0.015}, B=Multiply(A=@n24) {constB=0.1})) {constB=0}
  n26 = Rotator(Time=Constant() {r=3.5})
  n27 = OneMinus(None=MaterialFunctionCall(
    UVs=MaterialFunctionCall(UV Channel=@n26) {materialFunction=&LinearGradient}.UGradient
    Texture Scale=Constant() {r=1.5}
   ) {materialFunction=&ScaleUVsByCenter}.UVs)
  n28 = ScalarParameter() {defaultValue=5, parameterName=Corner Glow Power}
  n29 = Multiply(A=NamedRerouteUsage(), B=Saturate(None=Power(Base=@n27, Exp=@n28) {constExponent=5}))
  MaterialFunctionCall!(Emissive=Add(A=Multiply(A=@n7, B=@n8), B=Add(A=@n25, B=@n29)), Scanlines Tile) {materialFunction=/SparksMaps/Effects/Screens/Materials/MF/MF_Sparks_Screen_Details}
  n30 = VectorParameter() {desc="25, 0.1, 0.5, 0.3", parameterName=Streak Distort Tile And Speed, defaultValue=<rgba 10 0.1 0.5 0.3>}
  n31 = Multiply(A=Rotator(Coordinate=TextureCoordinate(), Time=Constant() {r=-3}), B=AppendVector(A=@n30.R, B=@n30.R))
  n32 = AppendVector(A=@n30.B, B=@n30.B)
  NamedRerouteDeclaration(None=TextureSample(UVs=Panner(Coordinate=@n31, Speed=@n32)) {samplerType=SAMPLERTYPE_Masks, texture=/Game/Global/Textures/Default/Noise/T_Metal_Noise}.R) {name=PaintDistortMovement, nodeColor=<rgba 0 0 0 1>}
  n33 = TextureCoordinate()
  n34 = Constant() {r=-3}
  NamedRerouteDeclaration(None=LinearInterpolate(
    A=Rotator(Coordinate=@n33, Time=@n34)
    B=NamedRerouteUsage()
    Alpha=ScalarParameter() {defaultValue=0.03, parameterName=Streak Distort Amount}
   ) {constAlpha=0.01}) {name=StreaksUVs, nodeColor=<rgba 0 0 0 1>}
  n35 = Multiply(A=NamedRerouteUsage(), B=AppendVector(A=Constant() {r=1}, B=Constant() {r=0.75}))
  NamedRerouteDeclaration(None=TextureSample(UVs=Panner(Coordinate=@n35) {speedY=0.125}) {samplerType=SAMPLERTYPE_LinearColor, texture=&/T_Sparks_Screen_SharpWave}.R) {name=StreaksRedChannel, nodeColor=<rgba 0 0 0 1>}
  n36 = Multiply(A=NamedRerouteUsage(), B=AppendVector(A=Constant() {r=1}, B=Constant() {r=0.25}))
  NamedRerouteDeclaration(None=TextureSample(UVs=Panner(Coordinate=@n36) {speedY=0.05}) {samplerType=SAMPLERTYPE_LinearColor, texture=&T_Sparks_Screen_SharpWave}.G) {name=StreaksGreenChannel, nodeColor=<rgba 0 0 0 1>}
  n37 = Multiply(A=NamedRerouteUsage(), B=AppendVector(A=Constant() {r=1}, B=Constant() {r=1}))
  NamedRerouteDeclaration(None=TextureSample(UVs=Panner(Coordinate=@n37) {speedY=0.085}) {samplerType=SAMPLERTYPE_LinearColor, texture=&T_Sparks_Screen_SharpWave}.B) {name=StreaksBlueChannel, nodeColor=<rgba 0 0 0 1>}
  n38 = MaterialFunctionCall!(CustomUPosition) {materialFunction=/SparksStageShow/Systems/StageManager/FXManager/MaterialFunctions/MF_MojoStageColors}
  NamedRerouteDeclaration(None=@n38.Primary) {name=PrimaryColor, nodeColor=<rgba 0 0 0 1>}
  NamedRerouteDeclaration(None=@n38.Secondary) {name=SecondaryColor, nodeColor=<rgba 0 0 0 1>}
  n39 = TextureSample(UVs=Panner(Coordinate=TextureCoordinate() {uTiling=0.5, vTiling=0.5}) {speedX=-0.25, speedY=0.25}) {samplerType=SAMPLERTYPE_Grayscale, texture=&LowResBlurredNoise}
  NamedRerouteDeclaration(None=LinearInterpolate(Alpha=Power(Base=@n39.R)) {constA=0.5, constB=5}) {name=StreakPanningGradient, nodeColor=<rgba 0 0 0 1>}
  n40 = CurveAtlasRowParameter(CurveTime=CollectionParameter() {collection=/FMAudioReactiveData/AudioAnalysis/MPC_Sparks_Audio_Parameters, parameterName=BeatLifetime}) {
   atlas=/SparksMaps/Effects/Screens/Curves/CurveAtlas_Audio_Beat_A
   curve=/SparksMaps/Effects/Screens/Curves/Curve_Audio_Beat_Remap_001
   parameterName=ReactivityCurve
  }
  NamedRerouteDeclaration(None=MaterialFunctionCall(Input=@n40, Input Low=Constant(), Input High=Constant() {r=1}, Target Low=Constant() {r=1}, Target High=Constant() {r=2}) {materialFunction=&RemapValueRange}.Result) {name=Reactivity, nodeColor=<rgba 0 0 0 1>}
  NamedRerouteUsage()
  VectorParameter() {parameterName=ColorB, defaultValue=<rgba 0.7292 0 0.2306 1>}
  VectorParameter() {parameterName=ColorA, defaultValue=<rgba 0 0.2409 0.7396 1>}
 ]
}
